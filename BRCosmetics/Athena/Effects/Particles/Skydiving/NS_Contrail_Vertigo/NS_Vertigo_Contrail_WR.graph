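# Niagara system for the Vertigo skydiving contrail cosmetic, attached to the character's skeletal mesh and tinted by a user color.
# Parent and BoneFlowOffset burst ribbons half a second apart, swirling along bones in opposite angular offsets via FN_BoneFlow.
# Drops_02 spawns sprites from Parent's particles, alongside mesh-sampled drops, curl-noise attached ribbons and Tendrill particles.
# Dive amount and a front-end preview flag adjust velocity, ribbon noise and lifetime.
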NiagaraSystem /BRCosmetics/Athena/Effects/Particles/Skydiving/NS_Contrail_Vertigo/NS_Vertigo_Contrail_WR
dependencies {
 usedDataInterfaces [
  &NiagaraDataInterfaceParticleRead
  &NiagaraDataInterfaceCurve
  &NiagaraDataInterfaceVectorCurve
  &NiagaraDataInterfaceColorCurve
 ]
 usedDynamicInputs [&FN_FrontEndFloatModifier, &SetBoolByFloatComparison]
 usedModules [
  &SystemState
  &self:SystemSpawnScript.SystemScriptSource.SystemScriptGraph.NiagaraNodeAssignment_0.SetVariables_#1
  &FN_Integrator
 ]
}
emitters {
 BoneFlowOffset {
  emitterSpawnScript [
   /SetVariables(Emitter.RandomSeed:NiagaraInt32=/Niagara/DynamicInputs/Random/RandomInt(Randomness Mode:ENiagaraRandomnessMode=::NewEnumerator0, Random Seed:NiagaraInt32=0, Fixed Random Seed:NiagaraBool=false)) {isSetParameters=true, module=SetVariables_#2}
   /SetVariables(Emitter.ParticlesPerGroup=6) {isSetParameters=true, module=SetVariables_#3}
  ]
  emitterUpdateScript [
   /SetVariables(Emitter.NumberOfGroups:NiagaraInt32=&RandomRangeInt(Minimum=2, Maximum=2, Recalculate Random Each Loop=false, Random Seed=$Emitter.LoopCount, Fixed Random Seed=true)) {isSetParameters=true, module=SetVariables_#4}
   &EmitterState(Life Cycle Mode=::NewEnumerator1, Loop Behavior=::NewEnumerator1, Loop Duration Mode=::NewEnumerator1, UseLoopDelay=false)
   &SpawnBurst_Instantaneous(
    Spawn Count=&Multiply_Int(A=$Emitter.ParticlesPerGroup, B=$Emitter.NumberOfGroups)
    Use Spawn Probability=false
   ) {enabled=false}
   &SpawnBurst_Instantaneous(
    Spawn Count=&Multiply_Int(A=$Emitter.ParticlesPerGroup, B=$Emitter.NumberOfGroups)
    Spawn Time=0.5
    Use Spawn Probability=false
   ) {module=SpawnBurst_Instantaneous001}
  ]
  particleSpawnScript [
   /SetVariables(Particles.LoopCount:NiagaraInt32=$Emitter.LoopCount) {isSetParameters=true, module=SetVariables_#5}
   &FN_SetSpawnGroup(
    Inclusivity=true
    Normalize Spawn Group=true
    Particles Per Group:NiagaraInt32=$Emitter.ParticlesPerGroup
    Pattern=::NewEnumerator0
    Set Ribbon ID and Ribbon Link Order=true
   )
   &InitializeRibbon(
    Color=$User.Color_Tint
    Module.Write Color=true
    Module.Write Lifetime=true
    Module.Write Mass=true
    Module.Write Material Random=true
    Module.Write Position=false
    Module.Write RibbonTwist=false
    Module.Write RibbonWidth=true
    Ribbon Width=1
   )
   &FN_FlipRibbonV()
  ]
  particleUpdateScript [
   &ParticleState(Kill Particles When Lifetime Has Elapsed=false, Let Infinitely Lived Particles Die When Emitter Deactivates=true)
   /Game/Effects/Niagara/Modules/BoneFlow/FN_BoneFlow_v4/FN_BoneFlow(
    Angle:NiagaraFloat=&Add_Float(A=$Particles.Age, B=&Lerp_Float(A=0, B=-1, Alpha=$Particles.Normalized Spawn Group))
    BoneID:NiagaraInt32=&ModuloInt(A=$Particles.SpawnGroup, B=2)
    Height:NiagaraFloat=$Particles.Normalized Spawn Group
    Interpolate Mesh Sampling=false
    Radius:NiagaraFloat=&FloatFromCurve(
     FloatCurve=DataInterface() {propertyValues {Curve {keys [
         {arriveTangent=6.6907, interpMode=RCIM_Cubic, leaveTangent=6.6907, tangentMode=RCTM_User, time=0, value=0}
         {interpMode=RCIM_Cubic, tangentMode=RCTM_User, time=0.5, value=1}
         {arriveTangent=-6.7167, interpMode=RCIM_Cubic, leaveTangent=-6.7167, tangentMode=RCTM_User, time=1, value=0}
        ]}}}
     CurveIndex=$Particles.Normalized Spawn Group
     Scale Curve=7.5
    )
    SkeletalMesh:NiagaraDataInterfaceSkeletalMesh=$User.SkeletalMesh
    Write Position Attribute=true
   ) {module=FN_BoneFlow001}
   /SetVariables(Particles.RibbonFacing=$Output.FN_BoneFlow001.SplineBinormal) {isSetParameters=true, module=SetVariables_#6}
   &FN_RibbonMaterial(
    Color=<rgba 0 1 1 9>
    Opacity Pan U Offset:NiagaraFloat=/Game/Effects/Niagara/DynamicInputs/Random/FN_SeededRandomRangeFloat(
     Minimum:NiagaraFloat=0
     Maximum:NiagaraFloat=2
     Random Seed:NiagaraInt32=&Add_Integer(A=$Particles.LoopCount, B=$Particles.SpawnGroup)
    )
    Opacity Pan U Speed:NiagaraFloat=&FN_SeededRandomRangeFloat(
     Minimum:NiagaraFloat=0.1364
     Maximum:NiagaraFloat=0.5182
     Random Seed:NiagaraInt32=&Add_Integer(A=$Particles.LoopCount, B=$Particles.SpawnGroup)
    )
    Opacity Scale U:NiagaraFloat=&FN_SeededRandomRangeFloat(
     Minimum:NiagaraFloat=0.4273
     Maximum:NiagaraFloat=1.0032
     Random Seed:NiagaraInt32=&Add_Integer(A=$Particles.LoopCount, B=$Particles.SpawnGroup)
    )
    Scale Alpha:NiagaraFloat=&FloatFromCurve(
     FloatCurve=DataInterface() {propertyValues {Curve {keys [<at -0.007 0.0112>,<at 0.2367 0.9629>,<at 0.7537 0.9995>,<at 1 0>]}}}
     Scale Curve=&FloatFromCurve(
      FloatCurve=DataInterface() {propertyValues {Curve {keys [<at 0.0007 0.022>,<at 0.2494 1.0225>,<at 0.8089 0.9748>,<at 1 0>]}}}
      CurveIndex=$Particles.Normalized Spawn Group
     )
    )
    UV Basis:NiagaraFloat=$Particles.Normalized Spawn Group
   ) {enabled=false}
   &ScaleRibbonWidth(Ribbon Width Scale=&FloatFromCurve(
     FloatCurve=DataInterface() {propertyValues {Curve {keys [
         {arriveTangent=4.1693, interpMode=RCIM_Cubic, leaveTangent=4.1693, tangentMode=RCTM_User, time=0, value=0}
         {interpMode=RCIM_Cubic, tangentMode=RCTM_User, time=0.5, value=1}
         {arriveTangent=-3.8576, interpMode=RCIM_Cubic, leaveTangent=-3.8576, tangentMode=RCTM_User, time=1, value=0}
        ]}}}
     CurveIndex=$Particles.Normalized Spawn Group
     Scale Curve=&Multiply_Float(
      A=2.35
      B=&FloatFromCurve(
       FloatCurve=DataInterface() {propertyValues {Curve {keys [{interpMode=RCIM_Cubic, tangentMode=RCTM_User, time=0, value=0}, {interpMode=RCIM_Cubic, tangentMode=RCTM_User, time=1, value=1}]}}}
       CurveIndex=$Particles.Age
      )
     )
    ))
   &DynamicMaterialParameters(
    Index 0 Param 1=&FloatFromCurve(FloatCurve=DataInterface() {propertyValues {Curve {keys [<at 0 0 RCIM_Cubic>,<at 0.5 1>,<at 1 0 RCIM_Cubic>]}}})
    Module.Param0WriteEnabled=false
    Module.Param1WriteEnabled=false
    Module.Param2WriteEnabled=false
    Module.Param3WriteEnabled=false
   )
  ]
  rendererClasses [&NiagaraRibbonRendererProperties]
  renderers [{rendererClass=&NiagaraRibbonRendererProperties}]
 }
 Drops_02 {
  emitterSpawnScript [&FN_PreserveSkydiveUserVariables() {module=PreserveSkydiveUserVariables}]
  emitterUpdateScript [
   &EmitterState()
   /Game/Effects/Niagara/Modules/Spawn/FN_SkydiveSpawnRate(Global Spawnrate=2) {enabled=false, module=SkydiveSpawnRate}
   &SpawnRate(SpawnRate=8, Use Spawn Probability=false) {enabled=false}
   &SpawnParticlesFromOtherEmitter(
    Attribute Reader=DataInterface() {propertyValues {EmitterBinding {bindingMode=Other, emitterName=Parent}}}
    Spawn Rate=2
    Use Maximum Age=false
    Use Minimum Age=false
   )
  ]
  particleSpawnScript [
   &FN_SetSpawnGroup(Number of Spawn Groups=6) {enabled=false, module=FN_SetSpawnGroup001}
   /Niagara/Modules/Spawn/Initialization/InitializeParticle(
    Color=$User.Color_Tint
    Lifetime=&UniformRangedFloat(Minimum=0.3, Maximum=0.7)
    Module.Write Color=true
    Module.Write Lifetime=true
    Module.Write Mass=false
    Module.Write Material Random=true
    Module.Write Position=false
    Module.Write Scale=false
    Module.Write SpriteRotation=false
    Module.Write SpriteSize=true
    Position=&SimulationPosition()
    Sprite Rotation=&UniformRangedFloat(Minimum=0, Maximum=1)
    Sprite Size=&Vector2DFromFloat(Value=&UniformRangedFloat(Minimum=3.75, Maximum=6.5))
   )
   &SampleParticlesFromOtherEmitter(Emitter Sampling Mode:ENiagara_EmitterLocSamplingMode=::NewEnumerator0)
   &SkeletalMeshLocation(
    Bone / Socket Index=$Particles.SpawnGroup
    Bone Sampling Mode=::NewEnumerator4
    Inherit Velocity=false
    MaskSpawnGroup=false
    Skeletal Mesh=$User.SkeletalMesh
   ) {enabled=false}
   /Niagara/Modules/Update/SubUV/SubUVAnimation(SubUV Animation Mode=::NewEnumerator1)
  ]
  particleUpdateScript [
   &ParticleState()
   /Game/Effects/Niagara/Modules/Skydiving/FN_DivingAttributes/v4/FN_DivingAttributes_v4(Bone Index=6, SkeletalMesh:NiagaraDataInterfaceSkeletalMesh=$User.SkeletalMesh) {enabled=false}
   &ScaleSpriteSize(Scale Factor=&Vector2DFromFloat(Value=&FloatFromCurve(FloatCurve=DataInterface() {propertyValues {Curve {keys [<at 0 0>,<at 0.0524 1>,<at 0.85 1>,<at 1.0011 0.002>]}}})))
   /SetVariables(Particles.Velocity=/Niagara/DynamicInputs/ScaleByCurve/ScaleVectorByCurve(
     InitialValue:Vector3f=&RandomRangeVector(Minimum=<-1.5 -1.5 0.7>, Maximum=<1.5 1.5 1.3>)
     ScaleCurve:NiagaraDataInterfaceVectorCurve=DataInterface() {propertyValues {
       XCurve {keys [<at 0 50 RCIM_Cubic>, {arriveTangent=-241.6722, interpMode=RCIM_Cubic, leaveTangent=-241.672, tangentMode=RCTM_User, time=0.5, value=20}]}
       YCurve {keys [<at 0 50 RCIM_Cubic>, {arriveTangent=-247.443, interpMode=RCIM_Cubic, leaveTangent=-247.448, tangentMode=RCTM_User, time=0.4991, value=20}]}
       ZCurve {keys [<at 0 100>,<at 1 300>]}
      }}
     CurveIndex:NiagaraFloat=$Particles.NormalizedAge
    )) {enabled=false, isSetParameters=true, module=SetVariables_#7}
   &InheritSourceMovement(
    Applied Force Scale=&VectorFromFloat(Value=&FloatFromCurve(FloatCurve=DataInterface() {propertyValues {Curve {keys [{arriveTangent=-1.6218, interpMode=RCIM_Cubic, leaveTangent=-1.6218, tangentMode=RCTM_User, time=0, value=0.7}, <at 1 0 RCIM_Cubic>]}}}))
    Applied Position Falloff Distance=750
    Applied Position Scale=&VectorFromFloat(Value=0.95)
    ApplyForce=false
    ApplyPosition=true
    ApplyPositionFalloff=true
    ApplyPositionFalloffCurve=true
   )
   &AddVelocity(Velocity=&MakeCustomVectorFromBool(
     Bool=$User.bIsFrontEndPreview
     True Vector=&Lerp_Vector(
      A=&RandomRangeVector(Minimum=<0 0 7>, Maximum=<0 0 13>)
      B=&RandomRangeVector(Minimum=<0 -5 13>, Maximum=<0 -15 18>)
      Alpha=$System.Dive
     )
     False Vector=<0 0 0>
    ))
   &SolveForcesAndVelocity(Clamp Velocity=false, Limit Acceleration=false)
   &SpriteRotationRate(Rotation Rate=$Particles.InitRotationRate)
  ]
  rendererClasses [&NiagaraSpriteRendererProperties]
  renderers [{rendererClass=&NiagaraSpriteRendererProperties}]
 }
 Drops_Local {
  emitterSpawnScript [&FN_PreserveSkydiveUserVariables() {module=PreserveSkydiveUserVariables}]
  emitterUpdateScript [
   &EmitterState()
   &FN_SkydiveSpawnRate(Global Spawnrate=2) {enabled=false, module=SkydiveSpawnRate}
   &SpawnRate(SpawnRate=18, Use Spawn Probability=false)
  ]
  particleSpawnScript [
   &FN_SetSpawnGroup(Number of Spawn Groups=6) {module=FN_SetSpawnGroup001}
   &InitializeParticle(
    Color=$User.Color_Tint
    Lifetime=&UniformRangedFloat(Minimum=0.3, Maximum=0.7)
    Module.Write Color=true
    Module.Write Lifetime=true
    Module.Write Mass=false
    Module.Write Material Random=true
    Module.Write Position=false
    Module.Write Scale=false
    Module.Write SpriteRotation=false
    Module.Write SpriteSize=true
    Position=&SimulationPosition()
    Sprite Rotation=&UniformRangedFloat(Minimum=0, Maximum=1)
    Sprite Size=&Vector2DFromFloat(Value=&UniformRangedFloat(Minimum=7, Maximum=10))
   )
   &SubUVAnimation(SubUV Animation Mode=::NewEnumerator1)
  ]
  particleUpdateScript [
   &ParticleState()
   &SkeletalMeshLocation(
    Bone / Socket Index=$Particles.SpawnGroup
    Bone Sampling Mode=::NewEnumerator4
    Inherit Velocity=false
    MaskSpawnGroup=false
    Skeletal Mesh=$User.SkeletalMesh
   ) {enabled=false}
   &SkeletalMeshLocation(
    Bone Sampling Mode=::NewEnumerator0
    MaskSpawnGroup=false
    OverrideRandomSeed=false
    Position Sampling=::NewEnumerator1
    Sampled Position Offset=&RandomRangeVector(Minimum=<-3 -3 -3>, Maximum=<3 3 3>)
    Skeletal Mesh=$User.SkeletalMesh
   ) {module=SkeletalMeshLocation001}
   &AccelerationForce(Acceleration=&RandomRangeVector(Minimum=<-45 -45 75>, Maximum=<45 45 150>))
   &ScaleSpriteSize(Scale Factor=&Vector2DFromFloat(Value=&FloatFromCurve(FloatCurve=DataInterface() {propertyValues {Curve {keys [<at 0 0>,<at 0.2 1>,<at 0.85 1>,<at 1.0011 0.002>]}}})))
   &SolveForcesAndVelocity(Clamp Velocity=false, Limit Acceleration=false)
   &SpriteRotationRate(Rotation Rate=$Particles.InitRotationRate)
   &CameraOffset(Camera Offset Amount=10)
  ]
  rendererClasses [&NiagaraSpriteRendererProperties]
  renderers [{rendererClass=&NiagaraSpriteRendererProperties}]
 }
 NE_Empty {
  emitterUpdateScript [
   &EmitterState()
   &SpawnRate(SpawnRate=2, Use Spawn Probability=false)
  ]
  particleSpawnScript [
   &FN_SetSpawnGroup(Number of Spawn Groups=6)
   &InitializeParticle(
    Lifetime=0.6
    Module.Write Color=true
    Module.Write Lifetime=true
    Module.Write Mass=true
    Module.Write Material Random=true
    Module.Write Position=false
    Module.Write Scale=false
    Module.Write SpriteRotation=false
    Module.Write SpriteSize=true
    Position=&SimulationPosition()
    Sprite Size=<26 26>
   )
   &SkeletalMeshLocation(
    Bone / Socket Index=$Particles.SpawnGroup
    Bone Sampling Mode=::NewEnumerator4
    Inherit Velocity=false
    MaskSpawnGroup=false
    Skeletal Mesh=$User.New Skeletal Mesh
   )
   /SetVariables(Particles.Velocity=&ScaleVectorByCurve(
     InitialValue:Vector3f=&RandomRangeVector(Minimum=<-1.5 -1.5 1.4>, Maximum=<1.5 1.5 3.4>)
     ScaleCurve:NiagaraDataInterfaceVectorCurve=DataInterface() {propertyValues {
       XCurve {keys [<at 0 50 RCIM_Cubic>, {arriveTangent=-241.6722, interpMode=RCIM_Cubic, leaveTangent=-241.672, tangentMode=RCTM_User, time=0.5, value=20}]}
       YCurve {keys [<at 0 50 RCIM_Cubic>, {arriveTangent=-247.443, interpMode=RCIM_Cubic, leaveTangent=-247.448, tangentMode=RCTM_User, time=0.4991, value=20}]}
       ZCurve {keys [<at 0 100>,<at 1 300>]}
      }}
     CurveIndex:NiagaraFloat=$Particles.NormalizedAge
    )) {isSetParameters=true}
   &AddVelocity(Velocity=&MakeCustomVectorFromBool(
     Bool=$User.bIsFrontEndPreview
     True Vector=&Lerp_Vector(A=<0 0 0>, B=<0 -100 100>, Alpha=$System.Dive)
     False Vector=<0 0 0>
    ))
  ]
  particleUpdateScript [
   &ParticleState()
   &SolveForcesAndVelocity(Clamp Velocity=false, Limit Acceleration=false)
   &Color(Color=&ColorFromCurve(ColorCurve=DataInterface() {propertyValues {
       AlphaCurve {keys [<at 0 1 RCIM_Cubic>,<at 1 0 RCIM_Cubic>]}
       BlueCurve {keys [<at 0 1 RCIM_Cubic>,<at 1 0 RCIM_Cubic>]}
       GreenCurve {keys [<at 0 1 RCIM_Cubic>,<at 1 0 RCIM_Cubic>]}
       RedCurve {keys [<at 0 1 RCIM_Cubic>,<at 1 0 RCIM_Cubic>]}
      }}))
  ]
  rendererClasses [&NiagaraSpriteRendererProperties]
  renderers [{rendererClass=&NiagaraSpriteRendererProperties}]
 }
 Parent {
  emitterSpawnScript [
   /SetVariables(Emitter.RandomSeed:NiagaraInt32=&RandomInt(Randomness Mode:ENiagaraRandomnessMode=::NewEnumerator0, Random Seed:NiagaraInt32=0, Fixed Random Seed:NiagaraBool=false)) {isSetParameters=true, module=SetVariables_#2}
   /SetVariables(Emitter.ParticlesPerGroup=6) {isSetParameters=true, module=SetVariables_#3}
  ]
  emitterUpdateScript [
   /SetVariables(Emitter.NumberOfGroups:NiagaraInt32=&RandomRangeInt(Minimum=2, Maximum=2, Recalculate Random Each Loop=false, Random Seed=$Emitter.LoopCount, Fixed Random Seed=true)) {isSetParameters=true, module=SetVariables_#4}
   &EmitterState(Life Cycle Mode=::NewEnumerator1, Loop Behavior=::NewEnumerator1, Loop Duration Mode=::NewEnumerator1, UseLoopDelay=false)
   &SpawnBurst_Instantaneous(
    Spawn Count=&Multiply_Int(A=$Emitter.ParticlesPerGroup, B=$Emitter.NumberOfGroups)
    Use Spawn Probability=false
   )
   &SpawnBurst_Instantaneous(
    Spawn Count=&Multiply_Int(A=$Emitter.ParticlesPerGroup, B=$Emitter.NumberOfGroups)
    Spawn Time=0.5
    Use Spawn Probability=false
   ) {enabled=false, module=SpawnBurst_Instantaneous001}
  ]
  particleSpawnScript [
   /SetVariables(Particles.LoopCount:NiagaraInt32=$Emitter.LoopCount) {isSetParameters=true, module=SetVariables_#5}
   &FN_SetSpawnGroup(
    Inclusivity=true
    Normalize Spawn Group=true
    Particles Per Group:NiagaraInt32=$Emitter.ParticlesPerGroup
    Pattern=::NewEnumerator0
    Set Ribbon ID and Ribbon Link Order=true
   )
   &InitializeRibbon(
    Color=$User.Color_Tint
    Module.Write Color=true
    Module.Write Lifetime=true
    Module.Write Mass=true
    Module.Write Material Random=true
    Module.Write Position=false
    Module.Write RibbonTwist=false
    Module.Write RibbonWidth=true
    Ribbon Width=2
   )
   &FN_FlipRibbonV()
  ]
  particleUpdateScript [
   &ParticleState(Kill Particles When Lifetime Has Elapsed=false, Let Infinitely Lived Particles Die When Emitter Deactivates=true)
   &FN_BoneFlow(
    Angle:NiagaraFloat=&Add_Float(A=$Particles.Age, B=&Lerp_Float(A=0, B=1, Alpha=$Particles.Normalized Spawn Group))
    BoneID:NiagaraInt32=&ModuloInt(A=$Particles.SpawnGroup, B=2)
    Height:NiagaraFloat=$Particles.Normalized Spawn Group
    Interpolate Mesh Sampling=false
    Radius:NiagaraFloat=&FloatFromCurve(
     FloatCurve=DataInterface() {propertyValues {Curve {keys [
         {arriveTangent=6.6907, interpMode=RCIM_Cubic, leaveTangent=6.6907, tangentMode=RCTM_User, time=0, value=0}
         {interpMode=RCIM_Cubic, tangentMode=RCTM_User, time=0.5, value=1}
         {arriveTangent=-6.7167, interpMode=RCIM_Cubic, leaveTangent=-6.7167, tangentMode=RCTM_User, time=1, value=0}
        ]}}}
     CurveIndex=$Particles.Normalized Spawn Group
     Scale Curve=7.5
    )
    SkeletalMesh:NiagaraDataInterfaceSkeletalMesh=$User.SkeletalMesh
    Write Position Attribute=true
   ) {module=FN_BoneFlow001}
   /SetVariables(Particles.RibbonFacing=$Output.FN_BoneFlow001.SplineBinormal) {isSetParameters=true, module=SetVariables_#6}
   &FN_RibbonMaterial(
    Color=<rgba 0 1 1 9>
    Opacity Pan U Offset:NiagaraFloat=&FN_SeededRandomRangeFloat(
     Minimum:NiagaraFloat=0
     Maximum:NiagaraFloat=2
     Random Seed:NiagaraInt32=&Add_Integer(A=$Particles.LoopCount, B=$Particles.SpawnGroup)
    )
    Opacity Pan U Speed:NiagaraFloat=&FN_SeededRandomRangeFloat(
     Minimum:NiagaraFloat=0.1364
     Maximum:NiagaraFloat=0.5182
     Random Seed:NiagaraInt32=&Add_Integer(A=$Particles.LoopCount, B=$Particles.SpawnGroup)
    )
    Opacity Scale U:NiagaraFloat=&FN_SeededRandomRangeFloat(
     Minimum:NiagaraFloat=0.4273
     Maximum:NiagaraFloat=1.0032
     Random Seed:NiagaraInt32=&Add_Integer(A=$Particles.LoopCount, B=$Particles.SpawnGroup)
    )
    Scale Alpha:NiagaraFloat=&FloatFromCurve(
     FloatCurve=DataInterface() {propertyValues {Curve {keys [<at -0.007 0.0112>,<at 0.2367 0.9629>,<at 0.7537 0.9995>,<at 1 0>]}}}
     Scale Curve=&FloatFromCurve(
      FloatCurve=DataInterface() {propertyValues {Curve {keys [<at 0.0007 0.022>,<at 0.2494 1.0225>,<at 0.8089 0.9748>,<at 1 0>]}}}
      CurveIndex=$Particles.Normalized Spawn Group
     )
    )
    UV Basis:NiagaraFloat=$Particles.Normalized Spawn Group
   ) {enabled=false}
   &ScaleRibbonWidth(Ribbon Width Scale=&FloatFromCurve(
     FloatCurve=DataInterface() {propertyValues {Curve {keys [
         {arriveTangent=4.1693, interpMode=RCIM_Cubic, leaveTangent=4.1693, tangentMode=RCTM_User, time=0, value=0}
         {interpMode=RCIM_Cubic, tangentMode=RCTM_User, time=0.5, value=1}
         {arriveTangent=-3.8576, interpMode=RCIM_Cubic, leaveTangent=-3.8576, tangentMode=RCTM_User, time=1, value=0}
        ]}}}
     CurveIndex=$Particles.Normalized Spawn Group
     Scale Curve=&Multiply_Float(
      A=2.35
      B=&FloatFromCurve(
       FloatCurve=DataInterface() {propertyValues {Curve {keys [{interpMode=RCIM_Cubic, tangentMode=RCTM_User, time=0, value=0}, {interpMode=RCIM_Cubic, tangentMode=RCTM_User, time=1, value=1}]}}}
       CurveIndex=$Particles.Age
      )
     )
    ))
   &DynamicMaterialParameters(
    Index 0 Param 1=&FloatFromCurve(FloatCurve=DataInterface() {propertyValues {Curve {keys [<at 0 0 RCIM_Cubic>,<at 0.5 1>,<at 1 0 RCIM_Cubic>]}}})
    Module.Param0WriteEnabled=false
    Module.Param1WriteEnabled=false
    Module.Param2WriteEnabled=false
    Module.Param3WriteEnabled=false
   )
  ]
  rendererClasses [&NiagaraRibbonRendererProperties]
  renderers [{rendererClass=&NiagaraRibbonRendererProperties}]
 }
 Ribbons_Attached {
  emitterUpdateScript [
   &EmitterState()
   &FN_SpawnRibbon_v2(Number of Ribbons=4, Spawn Mode=::NewEnumerator0, SpawnRate=30)
  ]
  particleSpawnScript [
   /Game/Effects/Niagara/Modules/Ribbons/FN_InitializeRibbon_v5(
    Color:LinearColor=$User.Color_Tint
    Enable Skeletal Position=true
    Ribbon Width=10
    SkeletalMesh:NiagaraDataInterfaceSkeletalMesh=$User.New Skeletal Mesh
   )
   &AddVelocity(Velocity=<0 0 100>) {enabled=false, module=AddVelocity002}
  ]
  particleUpdateScript [
   /Game/Effects/Niagara/Modules/Skydiving/FN_DivingAttributes/FN_DivingAttributes_v2(Accumulate Dive=0.1, Accumulate Rest=0.1, SkeletalMesh:NiagaraDataInterfaceSkeletalMesh=$User.New Skeletal Mesh) {module=FN_DivingAttributes_v2001}
   /SetVariables(Particles.Lifetime=&FloatFromCurve(
     FloatCurve=DataInterface() {propertyValues {Curve {keys [<at 0 0.35 RCIM_Constant>,<at 0.2 0.45 RCIM_Constant>,<at 0.35 0.55 RCIM_Constant>,<at 0.58 0.35 RCIM_Constant>,<at 0.8 0.45 RCIM_Constant>,<at 0.99 0.55 RCIM_Constant>]}}}
     CurveIndex=&NormalizeFloat(Value=&MakeFloatFromInt(Int=$Particles.SpawnGroup), Min=0, Max=5)
    )) {enabled=false, isSetParameters=true, module=SetVariables_#8}
   /SetVariables(Particles.Lifetime=&Multiply_Float(
     A=&FloatFromCurve(
      FloatCurve=DataInterface() {propertyValues {Curve {keys [<at 0 0.35 RCIM_Constant>,<at 0.2 0.45 RCIM_Constant>,<at 0.35 0.55 RCIM_Constant>,<at 0.58 0.35 RCIM_Constant>,<at 0.8 0.45 RCIM_Constant>,<at 0.99 0.55 RCIM_Constant>]}}}
      CurveIndex=&NormalizeFloat(Value=&MakeFloatFromInt(Int=$Particles.SpawnGroup), Min=0, Max=5)
     )
     B=&Lerp_Float(
      A=1
      B=&Lerp_Float(
       A=0.3
       B=0.7
       Alpha=&MakeCustomFloatFromBool(Bool=$User.bIsFrontEndPreview, True Float=1, False Float=0)
      )
      Alpha=$System.Dive
     )
    )) {isSetParameters=true, module=SetVariables_#9}
   &InheritVelocity(
    Module.Limit Speed=true
    Velocity Limit=10000
    Velocity Scale=&MakeCustomVectorFromBool(Bool=$Particles.Initial.Locked, True Vector=<0 0 0>, False Vector=<-0.03 -0.03 -0.007>)
    Velocity Source=&TransformVector(Vector=$Engine.Owner.Velocity, Source Space=::World, Destination Space=::Local)
   ) {enabled=false}
   &AddVelocity(Velocity=&MakeVector(
     X=0
     Y=0
     Z=&FloatFromCurve(FloatCurve=DataInterface() {propertyValues {Curve {keys [<at 0 1>,<at 0.1 0>]}}}, Scale Curve=25)
    ))
   /SetVariables(Particles.BoneSpeed:NiagaraFloat=&NormalizeFloat(Value=&VectorLength(Vector=$Engine.Owner.Velocity), Min=1000, Max=1100)) {isSetParameters=true, module=SetVariables_#10}
   /SetVariables(Particles.Lifetime=&FloatFromCurve(FloatCurve=DataInterface() {propertyValues {Curve {keys [<at 0 1>,<at 1 0.3>]}}}, CurveIndex=$Particles.BoneSpeed, Scale Curve=0.7)) {enabled=false, isSetParameters=true, module=SetVariables_#11}
   &CurlNoiseForce(
    Noise Strength=&FloatFromCurve(FloatCurve=DataInterface() {propertyValues {Curve {keys [<at 0 1>,<at 1 2>]}}}, CurveIndex=$Particles.BoneSpeed, Scale Curve=1000)
    Pan Noise=true
    Pan Noise Field=<15 15 15>
   ) {enabled=false}
   &CurlNoiseForce(
    Noise Strength=&Lerp_Float(
     A=&Lerp_Float(A=1000, B=700, Alpha=$System.Dive)
     B=&Lerp_Float(A=550, B=700, Alpha=$System.Dive)
     Alpha=&MakeCustomFloatFromBool(Bool=$User.bIsFrontEndPreview, True Float=1, False Float=0)
    )
    Pan Noise=true
    Pan Noise Field=<15 15 15>
   ) {module=CurlNoiseForce001}
   &InheritSourceMovement(
    Applied Force Scale=&VectorFromFloat(Value=&FloatFromCurve(FloatCurve=DataInterface() {propertyValues {Curve {keys [<at 0 0.5>,<at 1 0>]}}}))
    Applied Position Falloff Distance=700
    Applied Position Scale=&VectorFromFloat(Value=&FloatFromCurve(FloatCurve=DataInterface() {propertyValues {Curve {keys [<at 0 1 RCIM_Cubic>, {arriveTangent=-1.6169, interpMode=RCIM_Cubic, leaveTangent=-1.6169, tangentMode=RCTM_User, time=1, value=0}]}}}))
    ApplyForce=true
    ApplyForceFalloff=false
    ApplyPosition=true
    ApplyPositionFalloff=false
    Module.LimitForceSpeed=false
   ) {enabled=false}
   &InheritSourceMovement(
    Applied Force Scale=&VectorFromFloat(Value=&FloatFromCurve(FloatCurve=DataInterface() {propertyValues {Curve {keys [<at 0 0.25>,<at 1 0>]}}}))
    Applied Position Falloff Distance=700
    Applied Position Scale=&VectorFromFloat(Value=&FloatFromCurve(FloatCurve=DataInterface() {propertyValues {Curve {keys [<at 0 1 RCIM_Cubic>, {arriveTangent=-1.6169, interpMode=RCIM_Cubic, leaveTangent=-1.6169, tangentMode=RCTM_User, time=1, value=0.45}]}}}))
    ApplyForce=true
    ApplyForceFalloff=false
    ApplyPosition=true
    ApplyPositionFalloff=false
    Module.LimitForceSpeed=false
   ) {enabled=false, module=InheritSourceMovement001}
   &InheritSourceMovement(
    Applied Force Scale=&VectorFromFloat(Value=&Lerp_Float(
      A=&FloatFromCurve(FloatCurve=DataInterface() {propertyValues {Curve {keys [<at 0 1 RCIM_Cubic>, {arriveTangent=-1.6169, interpMode=RCIM_Cubic, leaveTangent=-1.6169, tangentMode=RCTM_User, time=1, value=0}]}}})
      B=&FloatFromCurve(FloatCurve=DataInterface() {propertyValues {Curve {keys [<at 0 0.25>,<at 1 0>]}}})
      Alpha=$Particles.BoneSpeed
     ))
    Applied Position Falloff Distance=700
    Applied Position Scale=&VectorFromFloat(Value=&Lerp_Float(
      A=&FloatFromCurve(FloatCurve=DataInterface() {propertyValues {Curve {keys [<at 0 1 RCIM_Cubic>, {arriveTangent=-1.6169, interpMode=RCIM_Cubic, leaveTangent=-1.6169, tangentMode=RCTM_User, time=1, value=0}]}}})
      B=&FloatFromCurve(FloatCurve=DataInterface() {propertyValues {Curve {keys [<at 0 1 RCIM_Cubic>, {arriveTangent=-1.6169, interpMode=RCIM_Cubic, leaveTangent=-1.6169, tangentMode=RCTM_User, time=1, value=0.45}]}}})
      Alpha=$Particles.BoneSpeed
     ))
    ApplyForce=true
    ApplyForceFalloff=false
    ApplyPosition=true
    ApplyPositionFalloff=false
    Module.LimitForceSpeed=false
   ) {enabled=false, module=InheritSourceMovement002}
   &InheritSourceMovement(
    Applied Force Scale=&VectorFromFloat(Value=&FloatFromCurve(FloatCurve=DataInterface() {propertyValues {Curve {keys [<at 0 0.7>,<at 1 0>]}}}))
    Applied Position Falloff Distance=700
    Applied Position Scale=&VectorFromFloat(Value=&FloatFromCurve(FloatCurve=DataInterface() {propertyValues {Curve {keys [<at 0 0.9 RCIM_Cubic>, {interpMode=RCIM_Cubic, leaveTangent=-3.6479, time=1, value=0.7}]}}}))
    ApplyForce=true
    ApplyForceFalloff=false
    ApplyPosition=true
    ApplyPositionFalloff=false
    Module.LimitForceSpeed=false
   ) {module=InheritSourceMovement003}
   &AddVelocity(Velocity=&MakeCustomVectorFromBool(
     Bool=$User.bIsFrontEndPreview
     True Vector=&Lerp_Vector(
      A=&RandomRangeVector(Minimum=<0 0 2>, Maximum=<0 0 3>)
      B=&RandomRangeVector(Minimum=<0 -10 15>, Maximum=<0 -15 20>)
      Alpha=$System.Dive
     )
     False Vector=<0 0 0>
    )) {module=AddVelocity001}
   &SolveForcesAndVelocity(Clamp Velocity=false, Limit Acceleration=false)
   /Game/Effects/Niagara/Modules/Ribbons/FN_UpdateRibbon_v5(Enable Skeletal Position=true, SkeletalMesh:NiagaraDataInterfaceSkeletalMesh=$User.New Skeletal Mesh)
   &DynamicMaterialParameters(
    Index 0 Param 1=&FloatFromCurve(FloatCurve=DataInterface() {propertyValues {Curve {keys [<at 0 1>,<at 1 0>]}}})
    Index 0 Param 2=$Particles.MaterialRandom
    Index 0 Param 3=&Lerp_Float(A=1, B=2, Alpha=$Particles.PlayerSpeed)
    Index 0 Param 4=1
    Module.Param0WriteEnabled=true
    Module.Param1WriteEnabled=true
    Module.Param2WriteEnabled=false
    Module.Param3WriteEnabled=true
   ) {enabled=false}
   &DynamicMaterialParameters(
    Index 0 Param 1=&FloatFromCurve(FloatCurve=DataInterface() {propertyValues {Curve {keys [<at 0 1 RCIM_Cubic>, {arriveTangent=-2.4197, interpMode=RCIM_Cubic, leaveTangent=-2.4197, tangentMode=RCTM_User, time=1, value=0}]}}})
    Index 0 Param 2=$Particles.MaterialRandom
    Module.Param0WriteEnabled=false
    Module.Param1WriteEnabled=true
    Module.Param2WriteEnabled=false
    Module.Param3WriteEnabled=false
   ) {module=DynamicMaterialParameters001}
   &ScaleRibbonWidth(Ribbon Width Scale=&Multiply_Float(
     A=&FloatFromCurve(
      FloatCurve=DataInterface() {propertyValues {Curve {keys [<at 0 0.6 RCIM_Constant>,<at 0.2 0.7 RCIM_Constant>,<at 0.35 0.5 RCIM_Constant>,<at 0.58 0.6 RCIM_Constant>,<at 0.8 0.7 RCIM_Constant>,<at 0.99 0.5 RCIM_Constant>]}}}
      CurveIndex=&NormalizeFloat(Value=&MakeFloatFromInt(Int=$Particles.SpawnGroup), Min=0, Max=5)
     )
     B=&FloatFromCurve(
      FloatCurve=DataInterface() {propertyValues {Curve {keys [
          {arriveTangent=-0.2092, interpMode=RCIM_Cubic, leaveTangent=-0.2092, tangentMode=RCTM_User, time=0.0048, value=1}
          {arriveTangent=-1.9945, interpMode=RCIM_Cubic, leaveTangent=-1.9945, tangentMode=RCTM_User, time=0.8644, value=0.4004}
          {arriveTangent=-2.7238, interpMode=RCIM_Cubic, leaveTangent=-2.7238, tangentMode=RCTM_User, time=1, value=0}
         ]}}}
      Scale Curve=0.6
     )
    )) {module=ScaleRibbonWidth001}
   &Color(Color=&LerpLinearColors(StartColor=<rgba 1 0 0.8353 1>, EndColor=<rgba 1 0.7383 0 1>, LerpFactor=$Particles.IntegratedDive)) {enabled=false}
  ]
  rendererClasses [&NiagaraRibbonRendererProperties, &NiagaraSpriteRendererProperties]
  renderers [
   {rendererClass=&NiagaraRibbonRendererProperties}
   {rendererClass=&NiagaraSpriteRendererProperties, rendererIndex=1}
  ]
 }
 Ribbons_Attached001 {
  bEnabled=false
  emitterUpdateScript [
   &EmitterState()
   &FN_SpawnRibbon_v2(Number of Ribbons=6, Spawn Mode=::NewEnumerator0, SpawnRate=30)
  ]
  particleSpawnScript [
   &FN_InitializeRibbon_v5(
    Color:LinearColor=$User.Color_Tint
    Enable Skeletal Position=true
    Ribbon Width=25
    SkeletalMesh:NiagaraDataInterfaceSkeletalMesh=$User.SkeletalMesh
   )
   &AddVelocity(Velocity=<0 0 100>) {enabled=false, module=AddVelocity002}
  ]
  particleUpdateScript [
   &FN_DivingAttributes_v2(Accumulate Dive=0.1, Accumulate Rest=0.1, SkeletalMesh:NiagaraDataInterfaceSkeletalMesh=$User.SkeletalMesh) {module=FN_DivingAttributes_v2001}
   /SetVariables(Particles.Lifetime=&FloatFromCurve(
     FloatCurve=DataInterface() {propertyValues {Curve {keys [<at 0 0.35 RCIM_Constant>,<at 0.2 0.45 RCIM_Constant>,<at 0.35 0.55 RCIM_Constant>,<at 0.58 0.35 RCIM_Constant>,<at 0.8 0.45 RCIM_Constant>,<at 0.99 0.55 RCIM_Constant>]}}}
     CurveIndex=&NormalizeFloat(Value=&MakeFloatFromInt(Int=$Particles.SpawnGroup), Min=0, Max=5)
    )) {enabled=false, isSetParameters=true, module=SetVariables_#8}
   /SetVariables(Particles.Lifetime=&Multiply_Float(
     A=&FloatFromCurve(
      FloatCurve=DataInterface() {propertyValues {Curve {keys [<at 0 0.35 RCIM_Constant>,<at 0.2 0.45 RCIM_Constant>,<at 0.35 0.55 RCIM_Constant>,<at 0.58 0.35 RCIM_Constant>,<at 0.8 0.45 RCIM_Constant>,<at 0.99 0.55 RCIM_Constant>]}}}
      CurveIndex=&NormalizeFloat(Value=&MakeFloatFromInt(Int=$Particles.SpawnGroup), Min=0, Max=5)
     )
     B=&Lerp_Float(
      A=1
      B=&Lerp_Float(
       A=0.3
       B=0.7
       Alpha=&MakeCustomFloatFromBool(Bool=$User.bIsFrontEndPreview, True Float=1, False Float=0)
      )
      Alpha=$System.Dive
     )
    )) {isSetParameters=true, module=SetVariables_#9}
   &InheritVelocity(
    Module.Limit Speed=true
    Velocity Limit=10000
    Velocity Scale=&MakeCustomVectorFromBool(Bool=$Particles.Initial.Locked, True Vector=<0 0 0>, False Vector=<-0.03 -0.03 -0.007>)
    Velocity Source=&TransformVector(Vector=$Engine.Owner.Velocity, Source Space=::World, Destination Space=::Local)
   ) {enabled=false}
   &AddVelocity(Velocity=&MakeVector(
     X=0
     Y=0
     Z=&FloatFromCurve(FloatCurve=DataInterface() {propertyValues {Curve {keys [<at 0 1>,<at 0.1 0>]}}}, Scale Curve=25)
    ))
   /SetVariables(Particles.BoneSpeed:NiagaraFloat=&NormalizeFloat(Value=&VectorLength(Vector=$Engine.Owner.Velocity), Min=1000, Max=1100)) {isSetParameters=true, module=SetVariables_#10}
   /SetVariables(Particles.Lifetime=&FloatFromCurve(FloatCurve=DataInterface() {propertyValues {Curve {keys [<at 0 1>,<at 1 0.3>]}}}, CurveIndex=$Particles.BoneSpeed, Scale Curve=0.7)) {enabled=false, isSetParameters=true, module=SetVariables_#11}
   &CurlNoiseForce(
    Noise Strength=&FloatFromCurve(FloatCurve=DataInterface() {propertyValues {Curve {keys [<at 0 1>,<at 1 2>]}}}, CurveIndex=$Particles.BoneSpeed, Scale Curve=1000)
    Pan Noise=true
    Pan Noise Field=<15 15 15>
   ) {enabled=false}
   &CurlNoiseForce(
    Noise Strength=&Lerp_Float(
     A=&Lerp_Float(A=1000, B=700, Alpha=$System.Dive)
     B=&Lerp_Float(A=550, B=700, Alpha=$System.Dive)
     Alpha=&MakeCustomFloatFromBool(Bool=$User.bIsFrontEndPreview, True Float=1, False Float=0)
    )
    Pan Noise=true
    Pan Noise Field=<15 15 15>
   ) {module=CurlNoiseForce001}
   &InheritSourceMovement(
    Applied Force Scale=&VectorFromFloat(Value=&FloatFromCurve(FloatCurve=DataInterface() {propertyValues {Curve {keys [<at 0 0.5>,<at 1 0>]}}}))
    Applied Position Falloff Distance=700
    Applied Position Scale=&VectorFromFloat(Value=&FloatFromCurve(FloatCurve=DataInterface() {propertyValues {Curve {keys [<at 0 1 RCIM_Cubic>, {arriveTangent=-1.6169, interpMode=RCIM_Cubic, leaveTangent=-1.6169, tangentMode=RCTM_User, time=1, value=0}]}}}))
    ApplyForce=true
    ApplyForceFalloff=false
    ApplyPosition=true
    ApplyPositionFalloff=false
    Module.LimitForceSpeed=false
   ) {enabled=false}
   &InheritSourceMovement(
    Applied Force Scale=&VectorFromFloat(Value=&FloatFromCurve(FloatCurve=DataInterface() {propertyValues {Curve {keys [<at 0 0.25>,<at 1 0>]}}}))
    Applied Position Falloff Distance=700
    Applied Position Scale=&VectorFromFloat(Value=&FloatFromCurve(FloatCurve=DataInterface() {propertyValues {Curve {keys [<at 0 1 RCIM_Cubic>, {arriveTangent=-1.6169, interpMode=RCIM_Cubic, leaveTangent=-1.6169, tangentMode=RCTM_User, time=1, value=0.45}]}}}))
    ApplyForce=true
    ApplyForceFalloff=false
    ApplyPosition=true
    ApplyPositionFalloff=false
    Module.LimitForceSpeed=false
   ) {enabled=false, module=InheritSourceMovement001}
   &InheritSourceMovement(
    Applied Force Scale=&VectorFromFloat(Value=&Lerp_Float(
      A=&FloatFromCurve(FloatCurve=DataInterface() {propertyValues {Curve {keys [<at 0 1 RCIM_Cubic>, {arriveTangent=-1.6169, interpMode=RCIM_Cubic, leaveTangent=-1.6169, tangentMode=RCTM_User, time=1, value=0}]}}})
      B=&FloatFromCurve(FloatCurve=DataInterface() {propertyValues {Curve {keys [<at 0 0.25>,<at 1 0>]}}})
      Alpha=$Particles.BoneSpeed
     ))
    Applied Position Falloff Distance=700
    Applied Position Scale=&VectorFromFloat(Value=&Lerp_Float(
      A=&FloatFromCurve(FloatCurve=DataInterface() {propertyValues {Curve {keys [<at 0 1 RCIM_Cubic>, {arriveTangent=-1.6169, interpMode=RCIM_Cubic, leaveTangent=-1.6169, tangentMode=RCTM_User, time=1, value=0}]}}})
      B=&FloatFromCurve(FloatCurve=DataInterface() {propertyValues {Curve {keys [<at 0 1 RCIM_Cubic>, {arriveTangent=-1.6169, interpMode=RCIM_Cubic, leaveTangent=-1.6169, tangentMode=RCTM_User, time=1, value=0.45}]}}})
      Alpha=$Particles.BoneSpeed
     ))
    ApplyForce=true
    ApplyForceFalloff=false
    ApplyPosition=true
    ApplyPositionFalloff=false
    Module.LimitForceSpeed=false
   ) {enabled=false, module=InheritSourceMovement002}
   &InheritSourceMovement(
    Applied Force Scale=&VectorFromFloat(Value=&FloatFromCurve(FloatCurve=DataInterface() {propertyValues {Curve {keys [<at 0 0.7>,<at 1 0>]}}}))
    Applied Position Falloff Distance=700
    Applied Position Scale=&VectorFromFloat(Value=&FloatFromCurve(FloatCurve=DataInterface() {propertyValues {Curve {keys [<at 0 0.9 RCIM_Cubic>, {interpMode=RCIM_Cubic, leaveTangent=-3.6479, time=1, value=0.7}]}}}))
    ApplyForce=true
    ApplyForceFalloff=false
    ApplyPosition=true
    ApplyPositionFalloff=false
    Module.LimitForceSpeed=false
   ) {module=InheritSourceMovement003}
   &AddVelocity(Velocity=&MakeCustomVectorFromBool(
     Bool=$User.bIsFrontEndPreview
     True Vector=&Lerp_Vector(
      A=&RandomRangeVector(Minimum=<0 0 2>, Maximum=<0 0 3>)
      B=&RandomRangeVector(Minimum=<0 -10 15>, Maximum=<0 -15 20>)
      Alpha=$System.Dive
     )
     False Vector=<0 0 0>
    )) {module=AddVelocity001}
   &SolveForcesAndVelocity(Clamp Velocity=false, Limit Acceleration=false)
   &FN_UpdateRibbon_v5(Enable Skeletal Position=true, SkeletalMesh:NiagaraDataInterfaceSkeletalMesh=$User.SkeletalMesh)
   &DynamicMaterialParameters(
    Index 0 Param 1=&FloatFromCurve(FloatCurve=DataInterface() {propertyValues {Curve {keys [<at 0 1>,<at 1 0>]}}})
    Index 0 Param 2=$Particles.MaterialRandom
    Index 0 Param 3=&Lerp_Float(A=1, B=2, Alpha=$Particles.PlayerSpeed)
    Index 0 Param 4=1
    Module.Param0WriteEnabled=true
    Module.Param1WriteEnabled=true
    Module.Param2WriteEnabled=false
    Module.Param3WriteEnabled=true
   ) {enabled=false}
   &DynamicMaterialParameters(
    Index 0 Param 1=&FloatFromCurve(FloatCurve=DataInterface() {propertyValues {Curve {keys [<at 0 1 RCIM_Cubic>, {arriveTangent=-2.4197, interpMode=RCIM_Cubic, leaveTangent=-2.4197, tangentMode=RCTM_User, time=1, value=0}]}}})
    Index 0 Param 2=$Particles.MaterialRandom
    Index 0 Param 4=1
    Module.Param0WriteEnabled=true
    Module.Param1WriteEnabled=true
    Module.Param2WriteEnabled=false
    Module.Param3WriteEnabled=true
   ) {module=DynamicMaterialParameters001}
   &ScaleRibbonWidth(Ribbon Width Scale=&Multiply_Float(
     A=&FloatFromCurve(
      FloatCurve=DataInterface() {propertyValues {Curve {keys [<at 0 0.6 RCIM_Constant>,<at 0.2 0.7 RCIM_Constant>,<at 0.35 0.5 RCIM_Constant>,<at 0.58 0.6 RCIM_Constant>,<at 0.8 0.7 RCIM_Constant>,<at 0.99 0.5 RCIM_Constant>]}}}
      CurveIndex=&NormalizeFloat(Value=&MakeFloatFromInt(Int=$Particles.SpawnGroup), Min=0, Max=5)
     )
     B=&FloatFromCurve(
      FloatCurve=DataInterface() {propertyValues {Curve {keys [
          {arriveTangent=-0.2092, interpMode=RCIM_Cubic, leaveTangent=-0.2092, tangentMode=RCTM_User, time=0.0048, value=1}
          {arriveTangent=-1.9945, interpMode=RCIM_Cubic, leaveTangent=-1.9945, tangentMode=RCTM_User, time=0.8644, value=0.4004}
          {arriveTangent=-2.7238, interpMode=RCIM_Cubic, leaveTangent=-2.7238, tangentMode=RCTM_User, time=1, value=0}
         ]}}}
      Scale Curve=0.6
     )
    )) {module=ScaleRibbonWidth001}
   &Color(Color=&LerpLinearColors(StartColor=<rgba 1 0 0.8353 1>, EndColor=<rgba 1 0.7383 0 1>, LerpFactor=$Particles.IntegratedDive)) {enabled=false}
  ]
  rendererClasses [&NiagaraRibbonRendererProperties, &NiagaraSpriteRendererProperties]
  renderers [
   {rendererClass=&NiagaraRibbonRendererProperties}
   {rendererClass=&NiagaraSpriteRendererProperties, rendererIndex=1}
  ]
 }
 Tendrill {
  emitterSpawnScript [&FN_PreserveSkydiveUserVariables() {module=PreserveSkydiveUserVariables}]
  emitterUpdateScript [
   &EmitterState()
   &SpawnRate(SpawnRate=1.5, Use Spawn Probability=false)
  ]
  particleSpawnScript [
   &FN_SetSpawnGroup() {module=FN_SetSpawnGroup001}
   &InitializeParticle(
    Color=$User.Color_Tint
    Lifetime=&UniformRangedFloat(Minimum=0.4, Maximum=0.5)
    Mesh Scale=&VectorFromFloat(Value=&RandomRangeFloat(Minimum=3, Maximum=4))
    Module.Write Color=true
    Module.Write Lifetime=true
    Module.Write Mass=false
    Module.Write Material Random=true
    Module.Write Position=false
    Module.Write Scale=true
    Module.Write SpriteRotation=false
    Module.Write SpriteSize=false
    Position=&SimulationPosition()
    Sprite Rotation=&UniformRangedFloat(Minimum=0, Maximum=1)
   )
   &SkeletalMeshLocation(
    Bone / Socket Index=$Particles.SpawnGroup
    Bone Sampling Mode=::NewEnumerator4
    Inherit Velocity=false
    MaskSpawnGroup=false
    Skeletal Mesh=$User.New Skeletal Mesh
   )
   &InitialMeshOrientation(Orientation Vector=<0 0 -1>, Use Orientation Vector=true, Use Rotation Vector=false)
   &self:DivideParticles_05(HowManyRenderers=3) {module=DivideParticles_05}
  ]
  particleUpdateScript [
   &ParticleState()
   /SetVariables(Particles.Velocity=&ScaleVectorByCurve(
     InitialValue:Vector3f=&RandomRangeVector(Minimum=<-1.5 -1.5 0.7>, Maximum=<1.5 1.5 1.7>)
     ScaleCurve:NiagaraDataInterfaceVectorCurve=DataInterface() {propertyValues {
       XCurve {keys [<at 0 50 RCIM_Cubic>, {arriveTangent=-241.6722, interpMode=RCIM_Cubic, leaveTangent=-241.672, tangentMode=RCTM_User, time=0.5, value=20}]}
       YCurve {keys [<at 0 50 RCIM_Cubic>, {arriveTangent=-247.443, interpMode=RCIM_Cubic, leaveTangent=-247.448, tangentMode=RCTM_User, time=0.4991, value=20}]}
       ZCurve {keys [<at 0 100>,<at 1 300>]}
      }}
     CurveIndex:NiagaraFloat=$Particles.NormalizedAge
    )) {enabled=false, isSetParameters=true, module=SetVariables_#7}
   &InheritSourceMovement(
    Applied Force Scale=&VectorFromFloat(Value=&FloatFromCurve(FloatCurve=DataInterface() {propertyValues {Curve {keys [<at 0 0.7>,<at 1 0>]}}}))
    Applied Position Falloff Distance=700
    Applied Position Scale=&VectorFromFloat(Value=&FloatFromCurve(FloatCurve=DataInterface() {propertyValues {Curve {keys [
          {arriveTangent=-1.2506, interpMode=RCIM_Cubic, leaveTangent=-1.2506, tangentMode=RCTM_User, time=0, value=1}
          {arriveTangent=-1.2984, interpMode=RCIM_Cubic, leaveTangent=-1.2984, tangentMode=RCTM_User, time=1, value=0}
         ]}}}))
    ApplyForce=true
    ApplyForceFalloff=false
    ApplyPosition=true
    ApplyPositionFalloff=false
    Module.LimitForceSpeed=false
   )
   &AddVelocity(Velocity=&MakeCustomVectorFromBool(
     Bool=$User.bIsFrontEndPreview
     True Vector=&Lerp_Vector(A=<0 0 25>, B=<1 -30 35>, Alpha=$System.Dive)
     False Vector=<0 0 0>
    ))
   &SolveForcesAndVelocity(Clamp Velocity=false, Limit Acceleration=false)
   &ScaleMeshSize(Scale Factor=&VectorFromFloat(Value=&FloatFromCurve(FloatCurve=DataInterface() {propertyValues {Curve {keys [
          {arriveTangent=4.3474, interpMode=RCIM_Cubic, leaveTangent=4.3474, tangentMode=RCTM_User, time=-0.0044, value=0.133}
          <at 0.2999 1.0087>
          {arriveTangent=-2.3342, interpMode=RCIM_Cubic, leaveTangent=-2.3342, tangentMode=RCTM_User, time=1, value=0}
         ]}}})))
   &DynamicMaterialParameters(
    Index 0 Param 1=&FloatFromCurve(FloatCurve=DataInterface() {propertyValues {Curve {keys [<at 0 0 RCIM_Cubic>,<at 0.5 1 RCIM_Cubic>,<at 1 0 RCIM_Cubic>]}}}, Scale Curve=1.25)
    Index 0 Param 2=&FloatFromCurve(
     FloatCurve=DataInterface() {propertyValues {Curve {keys [<at 0 1 RCIM_Cubic>,<at 0.5 0.1 RCIM_Cubic>,<at 1 1 RCIM_Cubic>]}}}
     Scale Curve=3.5
    )
    Index 0 Param 3=1
    Module.Param0WriteEnabled=true
    Module.Param1WriteEnabled=true
    Module.Param2WriteEnabled=true
    Module.Param3WriteEnabled=false
   )
  ]
  rendererClasses [&NiagaraMeshRendererProperties]
  renderers [
   {rendererClass=&NiagaraMeshRendererProperties, rendererIndex=0}
   {rendererClass=&NiagaraMeshRendererProperties, rendererIndex=1}
   {rendererClass=&NiagaraMeshRendererProperties, rendererIndex=2}
  ]
 }
 Tendrill_Spyro {
  emitterSpawnScript [&FN_PreserveSkydiveUserVariables() {module=PreserveSkydiveUserVariables}]
  emitterUpdateScript [
   &EmitterState()
   &SpawnRate(SpawnRate=2, Use Spawn Probability=false)
  ]
  particleSpawnScript [
   &FN_SetSpawnGroup(Number of Spawn Groups=6) {module=FN_SetSpawnGroup001}
   &InitializeParticle(
    Color=$User.Color_Tint
    Lifetime=&RandomRangeFloat(Minimum=0.3, Maximum=0.6)
    Mesh Scale=&RandomRangeVector(Minimum=<0.7 0.7 0.7>, Maximum=<1.7 1 1>)
    Module.Write Color=true
    Module.Write Lifetime=true
    Module.Write Mass=false
    Module.Write Material Random=true
    Module.Write Position=false
    Module.Write Scale=true
    Module.Write SpriteRotation=false
    Module.Write SpriteSize=false
    Position=&SimulationPosition()
    Sprite Rotation=&UniformRangedFloat(Minimum=0, Maximum=1)
   )
   &SkeletalMeshLocation(
    Bone / Socket Index=$Particles.SpawnGroup
    Bone Sampling Mode=::NewEnumerator4
    Inherit Velocity=false
    MaskSpawnGroup=false
    Skeletal Mesh=$User.New Skeletal Mesh
   )
   &InitialMeshOrientation(
    Rotation=&MakeVector(X=&RandomRangeFloat(Minimum=0, Maximum=1), Y=0, Z=0)
    Use Orientation Vector=true
    Use Rotation Vector=true
   )
  ]
  particleUpdateScript [
   &ParticleState()
   &InheritSourceMovement(
    Applied Force Scale=&VectorFromFloat(Value=&FloatFromCurve(FloatCurve=DataInterface() {propertyValues {Curve {keys [<at 0 0.7>,<at 1 0>]}}}))
    Applied Position Falloff Distance=700
    Applied Position Scale=&VectorFromFloat(Value=&FloatFromCurve(FloatCurve=DataInterface() {propertyValues {Curve {keys [
          {arriveTangent=-1.1743, interpMode=RCIM_Cubic, leaveTangent=-1.1743, tangentMode=RCTM_User, time=0, value=1}
          {arriveTangent=-1.4331, interpMode=RCIM_Cubic, leaveTangent=-1.4331, tangentMode=RCTM_User, time=1, value=0}
         ]}}}))
    ApplyForce=true
    ApplyForceFalloff=false
    ApplyPosition=true
    ApplyPositionFalloff=false
    Module.LimitForceSpeed=false
   )
   &AddVelocity(Velocity=&MakeCustomVectorFromBool(
     Bool=$User.bIsFrontEndPreview
     True Vector=&Lerp_Vector(A=<0 0 25>, B=<1 -20 23>, Alpha=$System.Dive)
     False Vector=<0 0 0>
    ))
   &SolveForcesAndVelocity(Clamp Velocity=false, Limit Acceleration=false)
   &ScaleMeshSize(Scale Factor=&VectorFromFloat(Value=&FloatFromCurve(FloatCurve=DataInterface() {propertyValues {Curve {keys [
          {arriveTangent=2.8611, interpMode=RCIM_Cubic, leaveTangent=2.8611, tangentMode=RCTM_User, time=0, value=0}
          <at 0.4735 0.9861>
          {arriveTangent=-2.3342, interpMode=RCIM_Cubic, leaveTangent=-2.3342, tangentMode=RCTM_User, time=1, value=0}
         ]}}})))
   &DynamicMaterialParameters(
    Index 0 Param 2=2
    Index 0 Param 4=1
    Module.Param0WriteEnabled=false
    Module.Param1WriteEnabled=true
    Module.Param2WriteEnabled=true
    Module.Param3WriteEnabled=true
   )
   &UpdateMeshOrientation(Rotation Coordinate Space=::NewEnumerator3, Rotation Rate=1.5, Rotation Vector=<1 0 0>)
  ]
  rendererClasses [&NiagaraMeshRendererProperties]
  renderers [{rendererClass=&NiagaraMeshRendererProperties, rendererIndex=0}]
 }
 Trail_Local_Space {
  bEnabled=false
  emitterUpdateScript [
   &EmitterState()
   /SetVariables(Emitter.EmitterReader=DataInterface() {propertyValues {EmitterBinding {bindingMode=Other, emitterName=NE_Empty}}}) {isSetParameters=true, module=SetVariables_#12}
   &SpawnRate(
    SpawnRate=&self:Spawn_by_Emitter_01(New Particle Attribute Reader:NiagaraDataInterfaceParticleRead=$Emitter.EmitterReader, Rate:NiagaraFloat=20)
    Use Spawn Probability=true
   )
  ]
  particleSpawnScript [
   &self:Sample_Location_01(New Particle Attribute Reader:NiagaraDataInterfaceParticleRead=$Emitter.EmitterReader) {module=Sample_Location_01}
   &InitializeRibbon(
    Color=$User.Color_Tint
    Lifetime=0.3
    Module.Write Color=true
    Module.Write Lifetime=true
    Module.Write Mass=true
    Module.Write Material Random=true
    Module.Write Position=false
    Module.Write RibbonTwist=false
    Module.Write RibbonWidth=true
    Ribbon Width=&FloatFromCurve(
     FloatCurve=DataInterface() {propertyValues {Curve {keys [
         <at 0 1 RCIM_Cubic>
         <at 0.624 0.9731>
         {arriveTangent=-1.8179, interpMode=RCIM_Cubic, leaveTangent=-1.8179, tangentMode=RCTM_User, time=1, value=0}
        ]}}}
     CurveIndex=&self:SampleNormalizedAge_01(Emitter:NiagaraDataInterfaceParticleRead=$Emitter.EmitterReader, SampleID:NiagaraInt32=$Particles.SampleID)
     Scale Curve=4
    )
   )
  ]
  particleUpdateScript [
   &ParticleState()
   /SetVariables(Particles.RibbonLife:NiagaraFloat=/Game/Effects/Niagara/DynamicInputs/Math/FN_Fit(
     Clamp:NiagaraBool=false
     Value:NiagaraFloat=&self:SampleColor_01(Emitter:NiagaraDataInterfaceParticleRead=$Emitter.EmitterReader, SampleID:NiagaraInt32=$Particles.SampleID)
     SourceMin:NiagaraFloat=0
     SourceMax:NiagaraFloat=1
     DestMin:NiagaraFloat=0
     DestMax:NiagaraFloat=0.4
    )) {enabled=false, isSetParameters=true, module=SetVariables_#13}
   &InheritVelocity(Module.Limit Speed=true, Velocity Limit=3000, Velocity Scale=<0.1 0.1 0.1>) {enabled=false}
   &ScaleColor(ScaleA=true, ScaleRGB=false)
   &ScaleRibbonWidth(Ribbon Width Scale=&FloatFromCurve(FloatCurve=DataInterface() {propertyValues {Curve {keys [<at 0 0 RCIM_Cubic>,<at 0.5 1>,<at 1 0 RCIM_Cubic>]}}}, Scale Curve=1.2))
   &Drag(Drag=1.5, Use Linear Drag=true, Use Rotational Drag=false) {enabled=false}
   &InheritSourceMovement(
    Applied Force Scale=<-1 -1 0>
    Applied Position Scale=<1 1 0.1>
    ApplyForce=true
    ApplyForceFalloff=false
    ApplyPosition=false
    Module.LimitForceSpeed=false
   )
   /Niagara/Modules/Update/Forces/CurlNoiseForce(Noise Frequency=0.1, Noise Strength=1000, Pan Noise=true, Pan Noise Field=<0 0 2>)
   &SolveForcesAndVelocity(Clamp Velocity=false, Limit Acceleration=false)
   &DynamicMaterialParameters(
    Index 0 Param 1=&FloatFromCurve(FloatCurve=DataInterface() {propertyValues {Curve {keys [<at 0 0 RCIM_Cubic>,<at 0.5 1>,<at 1 0 RCIM_Cubic>]}}})
    Module.Param0WriteEnabled=true
    Module.Param1WriteEnabled=false
    Module.Param2WriteEnabled=false
    Module.Param3WriteEnabled=false
   )
  ]
  rendererClasses [&NiagaraRibbonRendererProperties]
  renderers [{rendererClass=&NiagaraRibbonRendererProperties}]
 }
 Trail_Local_Space001 {
  emitterUpdateScript [
   &EmitterState()
   /SetVariables(Emitter.EmitterReader=DataInterface() {propertyValues {EmitterBinding {bindingMode=Other, emitterName=NE_Empty}}}) {isSetParameters=true, module=SetVariables_#12}
   &SpawnRate(
    SpawnRate=&self:Spawn_by_Emitter_01(New Particle Attribute Reader:NiagaraDataInterfaceParticleRead=$Emitter.EmitterReader, Rate:NiagaraFloat=20)
    Use Spawn Probability=true
   )
  ]
  particleSpawnScript [
   &self:Sample_Location_01(New Particle Attribute Reader:NiagaraDataInterfaceParticleRead=$Emitter.EmitterReader) {module=Sample_Location_01}
   &InitializeRibbon(
    Color=$User.Color_Tint
    Lifetime=0.3
    Module.Write Color=true
    Module.Write Lifetime=true
    Module.Write Mass=true
    Module.Write Material Random=true
    Module.Write Position=false
    Module.Write RibbonTwist=false
    Module.Write RibbonWidth=true
    Ribbon Width=&FloatFromCurve(
     FloatCurve=DataInterface() {propertyValues {Curve {keys [
         <at 0 1 RCIM_Cubic>
         <at 0.624 0.9731>
         {arriveTangent=-1.8179, interpMode=RCIM_Cubic, leaveTangent=-1.8179, tangentMode=RCTM_User, time=1, value=0}
        ]}}}
     CurveIndex=&self:SampleNormalizedAge_01(Emitter:NiagaraDataInterfaceParticleRead=$Emitter.EmitterReader, SampleID:NiagaraInt32=$Particles.SampleID)
     Scale Curve=2.5
    )
   )
  ]
  particleUpdateScript [
   &ParticleState()
   /SetVariables(Particles.RibbonLife:NiagaraFloat=&FN_Fit(
     Clamp:NiagaraBool=false
     Value:NiagaraFloat=&self:SampleColor_01(Emitter:NiagaraDataInterfaceParticleRead=$Emitter.EmitterReader, SampleID:NiagaraInt32=$Particles.SampleID)
     SourceMin:NiagaraFloat=0
     SourceMax:NiagaraFloat=1
     DestMin:NiagaraFloat=0
     DestMax:NiagaraFloat=0.4
    )) {enabled=false, isSetParameters=true, module=SetVariables_#13}
   &InheritVelocity(Module.Limit Speed=true, Velocity Limit=3000, Velocity Scale=<0.1 0.1 0.1>) {enabled=false}
   &ScaleColor(ScaleA=true, ScaleRGB=false)
   &ScaleRibbonWidth(Ribbon Width Scale=&FloatFromCurve(FloatCurve=DataInterface() {propertyValues {Curve {keys [<at 0 0 RCIM_Cubic>,<at 0.5 1>,<at 1 0 RCIM_Cubic>]}}}, Scale Curve=1.2))
   &Drag(Drag=1.5, Use Linear Drag=true, Use Rotational Drag=false) {enabled=false}
   &InheritSourceMovement(
    Applied Force Scale=<-1 -1 0>
    Applied Position Scale=<1 1 0.1>
    ApplyForce=true
    ApplyForceFalloff=false
    ApplyPosition=false
    Module.LimitForceSpeed=false
   )
   &CurlNoiseForce(Noise Frequency=0.1, Noise Strength=1000, Pan Noise=true, Pan Noise Field=<0 0 2>)
   &SolveForcesAndVelocity(Clamp Velocity=false, Limit Acceleration=false)
   &DynamicMaterialParameters(
    Index 0 Param 1=&FloatFromCurve(FloatCurve=DataInterface() {propertyValues {Curve {keys [<at 0 0 RCIM_Cubic>,<at 0.5 1>,<at 1 0 RCIM_Cubic>]}}})
    Module.Param0WriteEnabled=false
    Module.Param1WriteEnabled=false
    Module.Param2WriteEnabled=false
    Module.Param3WriteEnabled=false
   )
  ]
  rendererClasses [&NiagaraRibbonRendererProperties]
  renderers [{rendererClass=&NiagaraRibbonRendererProperties}]
 }
}
summary {
 systemName=NS_Vertigo_Contrail_WR
 emitters [
  {emitterName=Drops_02, rendererClasses [&NiagaraSpriteRendererProperties]}
  {
   bEnabled=false
   emitterName=Ribbons_Attached001
   rendererClasses [&NiagaraRibbonRendererProperties, &NiagaraSpriteRendererProperties]
  }
  {emitterName=Tendrill_Spyro, rendererClasses [&NiagaraMeshRendererProperties]}
  {emitterName=Tendrill, rendererClasses [&NiagaraMeshRendererProperties]}
  {emitterName=NE_Empty, rendererClasses [&NiagaraSpriteRendererProperties]}
  {bEnabled=false, emitterName=Trail_Local_Space, rendererClasses [&NiagaraRibbonRendererProperties]}
  {
   emitterName=Ribbons_Attached
   rendererClasses [&NiagaraRibbonRendererProperties, &NiagaraSpriteRendererProperties]
  }
  {emitterName=Trail_Local_Space001, rendererClasses [&NiagaraRibbonRendererProperties]}
  {emitterName=Drops_Local, rendererClasses [&NiagaraSpriteRendererProperties]}
  {emitterName=Parent, rendererClasses [&NiagaraRibbonRendererProperties]}
  {emitterName=BoneFlowOffset, rendererClasses [&NiagaraRibbonRendererProperties]}
 ]
 userVariables [
  User.bIsFrontEnd {type=NiagaraBool}
  User.bIsFrontEndPreview {type=NiagaraBool}
  User.Color_Tint {type=LinearColor, description="", defaultValue {struct=LinearColor, value=<rgba 0.4566 0.4566 0.4566 1>}}
  User.New Skeletal Mesh {
   type=NiagaraDataInterfaceSkeletalMesh
   defaultValue {
    struct=NiagaraExt_VariableValue_DataInterface
    value {dataInterface=&self:NiagaraDataInterfaceSkeletalMesh_1}
   }
  }
  User.ParaGlideLeanAlpha {type=NiagaraFloat, defaultValue=0, description=""}
  User.PlayerSpeed {type=NiagaraFloat, defaultValue=0, description=""}
  User.RibbonIDOffset {type=NiagaraInt32}
  User.SkeletalMesh {
   type=NiagaraDataInterfaceSkeletalMesh
   defaultValue {
    struct=NiagaraExt_VariableValue_DataInterface
    value {dataInterface=&self:NiagaraDataInterfaceSkeletalMesh_0}
   }
  }
 ]
}
